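; Gameplay ability for the TidalCrane swarm boss, named as the AI-takeover cheat variant of triggering reinforcements.
; CanActivateAbility is overridden to always return true, with no checks of its own.
; The activate and end events hold only disabled placeholder parent calls, so this graph adds no reinforcement logic itself.
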(event Ability|EventActivateAbility
  (|CallParentFunction) ; This node is disabled and will not be called.\nDrag off pins to build functionality.
)

(event Ability|EventOnEndAbility
  (params (bWasCancelled :type Boolean))
  (|CallParentFunction bWasCancelled) ; This node is disabled and will not be called.\nDrag off pins to build functionality.
)

(fn CanActivateAbility
  (params (ActorInfo :type Gameplay%20Ability%20Actor%20Info%20Structure) (Handle :type Gameplay%20Ability%20Spec%20Handle%20Structure))
  (return 0 true))
